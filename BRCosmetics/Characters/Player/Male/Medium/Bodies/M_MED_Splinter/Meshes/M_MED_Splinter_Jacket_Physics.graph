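# Physics asset for the Splinter character's jacket, defining the bodies and joints that simulate its trench coat skirt against the lower body.
# Six five-box panel chains hang from tiny MSTR anchors on the pelvis, with limited swing and tight twist.
# Angularly free links tie some neighboring left and back panels sideways at staggered depths.
# Pelvis, leg and foot shapes form the body, alongside two short coat capsule chains on the pelvis.
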PhysicsAsset /BRCosmetics/Characters/Player/Male/Medium/Bodies/M_MED_Splinter/Meshes/M_MED_Splinter_Jacket_Physics
bodies {
 calf_l [{shapeType=Capsule, length=39.413, radius=8.8016, center=<-20.7199 -2.5182 -0.0527>, rotation=<rot 87.8253 -5.5959 0>}]
 calf_r [
  {shapeType=Capsule, length=40, radius=7.8758, center=<20.5325 1.9112 0.7854>, rotation=<rot 90 3.5763 8.0104>}
  {
   shapeType=Box
   extentX=22.375
   extentY=15.4375
   extentZ=18.125
   shapeName=calf_r_box
   center=<41.3756 0.3242 0.3129>
   rotation=<rot 0 -5.0001 0>
  }
 ]
 dyn_coat_l_02 [{shapeType=Capsule, length=8, radius=7, shapeName=dyn_coat_l_02_capsule, center=<5 5 0>, rotation=<rot 90 -90 -95>}]
 dyn_coat_l_03 [{shapeType=Capsule, length=8, radius=7, shapeName=dyn_coat_l_03_capsule, center=<5 5 0>, rotation=<rot 90 -90 -95>}]
 dyn_coat_r_02 [{shapeType=Capsule, length=8, radius=7, shapeName=dyn_coat_r_02_capsule, center=<5 -5 0>, rotation=<rot 90 14.0363 19.0362>}]
 dyn_coat_r_03 [{shapeType=Capsule, length=8, radius=7, shapeName=dyn_coat_r_03_capsule, center=<5 -5 0>, rotation=<rot 90 14.0363 19.0362>}]
 dyn_trenchCoat_bl_01 [{shapeType=Box, extentX=15, extentY=23.5745, extentZ=6.75, center=<5 0 0>}]
 dyn_trenchCoat_bl_02 [{shapeType=Box, extentX=15, extentY=23.5745, extentZ=6.25, center=<5 0 0>}]
 dyn_trenchCoat_bl_03 [{shapeType=Box, extentX=15, extentY=23.5745, extentZ=5, center=<5 0 0>}]
 dyn_trenchCoat_bl_04 [{shapeType=Box, extentX=15, extentY=23.5745, extentZ=5, center=<5 0 0>}]
 dyn_trenchCoat_bl_05 [{shapeType=Box, extentX=15, extentY=23.5745, extentZ=5, center=<5 0 0>}]
 dyn_trenchCoat_bl_MSTR [{shapeType=Capsule, length=0.505, radius=0.505, rotation=<rot 0 -90 0>}]
 dyn_trenchCoat_br_01 [{shapeType=Box, extentX=15, extentY=23.5745, extentZ=6.75, center=<5 0 0>}]
 dyn_trenchCoat_br_02 [{shapeType=Box, extentX=15, extentY=23.5745, extentZ=6.25, center=<5 0 0>}]
 dyn_trenchCoat_br_03 [{shapeType=Box, extentX=15, extentY=23.5745, extentZ=5, center=<5 0 0>}]
 dyn_trenchCoat_br_04 [{shapeType=Box, extentX=15, extentY=25.4238, extentZ=7, center=<5 0 0>}]
 dyn_trenchCoat_br_05 [{shapeType=Box, extentX=15, extentY=25.4238, extentZ=7, center=<5 0 0>}]
 dyn_trenchCoat_br_MSTR [{shapeType=Capsule, length=0.505, radius=0.505, rotation=<rot 0 -90 0>}]
 dyn_trenchCoat_fl_01 [{shapeType=Box, extentX=15, extentY=26.0272, extentZ=4.9879, center=<5 0 0>}]
 dyn_trenchCoat_fl_02 [{shapeType=Box, extentX=15, extentY=26.0272, extentZ=4.9879, center=<5 0 0>}]
 dyn_trenchCoat_fl_03 [{shapeType=Box, extentX=15, extentY=26.0272, extentZ=4.9879, center=<5 0 0>}]
 dyn_trenchCoat_fl_04 [{shapeType=Box, extentX=15, extentY=26.0272, extentZ=4.9879, center=<5 0 0>}]
 dyn_trenchCoat_fl_05 [{shapeType=Box, extentX=15, extentY=26.0272, extentZ=4.9879, center=<5 0 0>}]
 dyn_trenchCoat_fl_MSTR [{shapeType=Capsule, length=0.505, radius=0.505, rotation=<rot 0 -90 0>}]
 dyn_trenchCoat_fr_01 [{shapeType=Box, extentX=15, extentY=23.5745, extentZ=4, center=<5 0 0>}]
 dyn_trenchCoat_fr_02 [{shapeType=Box, extentX=15, extentY=23.5745, extentZ=5.9735, center=<5 0.0001 -0.2591>}]
 dyn_trenchCoat_fr_03 [{shapeType=Box, extentX=15, extentY=23.5745, extentZ=6.814, center=<5 0 0>}]
 dyn_trenchCoat_fr_04 [{shapeType=Box, extentX=14.0244, extentY=21.2126, extentZ=4.1686, center=<5 0 0>}]
 dyn_trenchCoat_fr_05 [{shapeType=Box, extentX=15, extentY=23.5745, extentZ=4.6906, center=<5 0 0>}]
 dyn_trenchCoat_fr_MSTR [{shapeType=Capsule, length=0.505, radius=0.505, rotation=<rot 0 -90 0>}]
 dyn_trenchCoat_l_01 [{shapeType=Box, extentX=15, extentY=26.0272, extentZ=4, center=<5 0 0>}]
 dyn_trenchCoat_l_02 [{shapeType=Box, extentX=15, extentY=26.0272, extentZ=4, center=<5 0 0>}]
 dyn_trenchCoat_l_03 [{shapeType=Box, extentX=15, extentY=26.0272, extentZ=4, center=<5 0 0>}]
 dyn_trenchCoat_l_04 [{shapeType=Box, extentX=15, extentY=26.0272, extentZ=4, center=<5 0 0>}]
 dyn_trenchCoat_l_05 [{shapeType=Box, extentX=15, extentY=26.0272, extentZ=4, center=<5 0 0>}]
 dyn_trenchCoat_l_MSTR [{shapeType=Capsule, length=0.505, radius=0.505, rotation=<rot 0 -90 0>}]
 dyn_trenchCoat_r_01 [{shapeType=Box, extentX=15, extentY=23.5745, extentZ=4, center=<5 0 0>}]
 dyn_trenchCoat_r_02 [{shapeType=Box, extentX=15, extentY=23.5745, extentZ=4, center=<5 0 0>}]
 dyn_trenchCoat_r_03 [{shapeType=Box, extentX=15, extentY=23.5745, extentZ=4, center=<5 0 0>}]
 dyn_trenchCoat_r_04 [{shapeType=Box, extentX=15, extentY=25.4806, extentZ=4, center=<5 0 0>}]
 dyn_trenchCoat_r_05 [{shapeType=Box, extentX=15, extentY=25.4806, extentZ=4, center=<5 0 0>, rotation=<rot 0 -0.0003 0.0001>}]
 dyn_trenchCoat_r_MSTR [{shapeType=Capsule, length=0.505, radius=0.505, rotation=<rot 0 -90 0>}]
 foot_l [{shapeType=Capsule, length=10.3948, radius=7.8627, center=<-5 5 0>, rotation=<rot 0 0 90>}]
 foot_r [{shapeType=Capsule, length=10.4138, radius=7.8958, center=<5 -5 0>, rotation=<rot 0 0 90>}]
 pelvis [
  {shapeType=Capsule, length=10.6573, radius=8.5058, center=<0 -1.2333 0>}
  {shapeType=Box, extentX=38.375, extentY=4.5, extentZ=27.3125, shapeName=pelvis_box, center=<-16.3839 -8 0>}
 ]
 spine_01 [{shapeType=Capsule, length=0, radius=2}]
 thigh_l [{shapeType=Capsule, length=30.4478, radius=9.7302, center=<-15.0003 0.7169 0.3545>, rotation=<rot 90 0 0>}]
 thigh_r [{shapeType=Capsule, length=33.5944, radius=9.1882, center=<15.0001 -0.2192 -1.258>, rotation=<rot 90 0 0>}]
}
constraints [
 {bone1Name=spine_01, bone2Name=pelvis, swing1LimitDegrees=45, swing2LimitDegrees=45, twistLimitDegrees=45}
 {bone1Name=thigh_l, bone2Name=pelvis, swing1LimitDegrees=45, swing2LimitDegrees=45, twistLimitDegrees=45}
 {bone1Name=calf_l, bone2Name=thigh_l, swing1LimitDegrees=45, swing2LimitDegrees=45, twistLimitDegrees=45}
 {bone1Name=thigh_r, bone2Name=pelvis, swing1LimitDegrees=45, swing2LimitDegrees=45, twistLimitDegrees=45}
 {bone1Name=calf_r, bone2Name=thigh_r, swing1LimitDegrees=45, swing2LimitDegrees=45, twistLimitDegrees=45}
 {bone1Name=dyn_trenchCoat_fl_01, bone2Name=dyn_trenchCoat_fl_MSTR, swing1LimitDegrees=35, swing2LimitDegrees=35, twistLimitDegrees=5}
 {bone1Name=dyn_trenchCoat_fl_02, bone2Name=dyn_trenchCoat_fl_01, swing1LimitDegrees=35, swing2LimitDegrees=35, twistLimitDegrees=5}
 {bone1Name=dyn_trenchCoat_fl_03, bone2Name=dyn_trenchCoat_fl_02, swing1LimitDegrees=35, swing2LimitDegrees=35, twistLimitDegrees=5}
 {bone1Name=dyn_trenchCoat_fl_04, bone2Name=dyn_trenchCoat_fl_03, swing1LimitDegrees=45, swing2LimitDegrees=45, twistLimitDegrees=5}
 {bone1Name=dyn_trenchCoat_fl_05, bone2Name=dyn_trenchCoat_fl_04, swing1LimitDegrees=45, swing2LimitDegrees=45, twistLimitDegrees=5}
 {bone1Name=dyn_trenchCoat_l_01, bone2Name=dyn_trenchCoat_l_MSTR, swing1LimitDegrees=35, swing2LimitDegrees=35, twistLimitDegrees=5}
 {bone1Name=dyn_trenchCoat_l_02, bone2Name=dyn_trenchCoat_l_01, swing1LimitDegrees=35, swing2LimitDegrees=35, twistLimitDegrees=5}
 {bone1Name=dyn_trenchCoat_l_03, bone2Name=dyn_trenchCoat_l_02, swing1LimitDegrees=35, swing2LimitDegrees=35, twistLimitDegrees=5}
 {bone1Name=dyn_trenchCoat_l_04, bone2Name=dyn_trenchCoat_l_03, swing1LimitDegrees=45, swing2LimitDegrees=45, twistLimitDegrees=5}
 {bone1Name=dyn_trenchCoat_l_05, bone2Name=dyn_trenchCoat_l_04, swing1LimitDegrees=45, swing2LimitDegrees=45, twistLimitDegrees=5}
 {bone1Name=dyn_trenchCoat_bl_01, bone2Name=dyn_trenchCoat_bl_MSTR, swing1LimitDegrees=35, swing2LimitDegrees=35, twistLimitDegrees=5}
 {bone1Name=dyn_trenchCoat_bl_02, bone2Name=dyn_trenchCoat_bl_01, swing1LimitDegrees=35, swing2LimitDegrees=35, twistLimitDegrees=5}
 {bone1Name=dyn_trenchCoat_bl_03, bone2Name=dyn_trenchCoat_bl_02, swing1LimitDegrees=35, swing2LimitDegrees=35, twistLimitDegrees=5}
 {bone1Name=dyn_trenchCoat_bl_04, bone2Name=dyn_trenchCoat_bl_03, swing1LimitDegrees=45, swing2LimitDegrees=45, twistLimitDegrees=5}
 {bone1Name=dyn_trenchCoat_bl_05, bone2Name=dyn_trenchCoat_bl_04, swing1LimitDegrees=45, swing2LimitDegrees=45, twistLimitDegrees=5}
 {bone1Name=dyn_trenchCoat_fr_01, bone2Name=dyn_trenchCoat_fr_MSTR, swing1LimitDegrees=35, swing2LimitDegrees=35, twistLimitDegrees=5}
 {bone1Name=dyn_trenchCoat_fr_02, bone2Name=dyn_trenchCoat_fr_01, swing1LimitDegrees=35, swing2LimitDegrees=35, twistLimitDegrees=5}
 {bone1Name=dyn_trenchCoat_fr_03, bone2Name=dyn_trenchCoat_fr_02, swing1LimitDegrees=35, swing2LimitDegrees=35, twistLimitDegrees=5}
 {bone1Name=dyn_trenchCoat_fr_04, bone2Name=dyn_trenchCoat_fr_03, swing1LimitDegrees=45, swing2LimitDegrees=45, twistLimitDegrees=5}
 {bone1Name=dyn_trenchCoat_fr_05, bone2Name=dyn_trenchCoat_fr_04, swing1LimitDegrees=45, swing2LimitDegrees=45, twistLimitDegrees=5}
 {bone1Name=dyn_trenchCoat_r_01, bone2Name=dyn_trenchCoat_r_MSTR, swing1LimitDegrees=35, swing2LimitDegrees=35, twistLimitDegrees=5}
 {bone1Name=dyn_trenchCoat_r_02, bone2Name=dyn_trenchCoat_r_01, swing1LimitDegrees=35, swing2LimitDegrees=35, twistLimitDegrees=5}
 {bone1Name=dyn_trenchCoat_r_03, bone2Name=dyn_trenchCoat_r_02, swing1LimitDegrees=35, swing2LimitDegrees=35, twistLimitDegrees=5}
 {bone1Name=dyn_trenchCoat_r_04, bone2Name=dyn_trenchCoat_r_03, swing1LimitDegrees=45, swing2LimitDegrees=45, twistLimitDegrees=5}
 {bone1Name=dyn_trenchCoat_r_05, bone2Name=dyn_trenchCoat_r_04, swing1LimitDegrees=45, swing2LimitDegrees=45, twistLimitDegrees=5}
 {bone1Name=dyn_trenchCoat_br_01, bone2Name=dyn_trenchCoat_br_MSTR, swing1LimitDegrees=35, swing2LimitDegrees=35, twistLimitDegrees=5}
 {bone1Name=dyn_trenchCoat_br_02, bone2Name=dyn_trenchCoat_br_01, swing1LimitDegrees=35, swing2LimitDegrees=35, twistLimitDegrees=5}
 {bone1Name=dyn_trenchCoat_br_03, bone2Name=dyn_trenchCoat_br_02, swing1LimitDegrees=35, swing2LimitDegrees=35, twistLimitDegrees=5}
 {bone1Name=dyn_trenchCoat_br_04, bone2Name=dyn_trenchCoat_br_03, swing1LimitDegrees=5, swing2LimitDegrees=45, twistLimitDegrees=5}
 {bone1Name=dyn_trenchCoat_br_05, bone2Name=dyn_trenchCoat_br_04, swing1LimitDegrees=5, swing2LimitDegrees=45, twistLimitDegrees=5}
 {
  bone1Name=dyn_trenchCoat_l_05
  bone2Name=dyn_trenchCoat_fl_05
  swing1LimitDegrees=45
  swing1Motion=Free
  swing2LimitDegrees=45
  swing2Motion=Free
  twistLimitDegrees=45
  twistMotion=Free
 }
 {
  bone1Name=dyn_trenchCoat_bl_04
  bone2Name=dyn_trenchCoat_l_04
  swing1LimitDegrees=45
  swing1Motion=Free
  swing2LimitDegrees=45
  swing2Motion=Free
  twistLimitDegrees=45
  twistMotion=Free
 }
 {
  bone1Name=dyn_trenchCoat_br_03
  bone2Name=dyn_trenchCoat_bl_03
  swing1LimitDegrees=45
  swing1Motion=Free
  swing2LimitDegrees=45
  swing2Motion=Free
  twistLimitDegrees=45
  twistMotion=Free
 }
 {
  bone1Name=dyn_trenchCoat_fl_MSTR
  bone2Name=pelvis
  swing1LimitDegrees=45
  swing1Motion=Free
  swing2LimitDegrees=45
  swing2Motion=Free
  twistLimitDegrees=45
  twistMotion=Free
 }
 {
  bone1Name=dyn_trenchCoat_l_MSTR
  bone2Name=pelvis
  swing1LimitDegrees=45
  swing1Motion=Free
  swing2LimitDegrees=45
  swing2Motion=Free
  twistLimitDegrees=45
  twistMotion=Free
 }
 {
  bone1Name=dyn_trenchCoat_bl_MSTR
  bone2Name=pelvis
  swing1LimitDegrees=45
  swing1Motion=Free
  swing2LimitDegrees=45
  swing2Motion=Free
  twistLimitDegrees=45
  twistMotion=Free
 }
 {
  bone1Name=dyn_trenchCoat_fr_MSTR
  bone2Name=pelvis
  swing1LimitDegrees=45
  swing1Motion=Free
  swing2LimitDegrees=45
  swing2Motion=Free
  twistLimitDegrees=45
  twistMotion=Free
 }
 {
  bone1Name=dyn_trenchCoat_r_MSTR
  bone2Name=pelvis
  swing1LimitDegrees=45
  swing1Motion=Free
  swing2LimitDegrees=45
  swing2Motion=Free
  twistLimitDegrees=45
  twistMotion=Free
 }
 {
  bone1Name=dyn_trenchCoat_br_MSTR
  bone2Name=pelvis
  swing1LimitDegrees=45
  swing1Motion=Free
  swing2LimitDegrees=45
  swing2Motion=Free
  twistLimitDegrees=45
  twistMotion=Free
 }
 {
  bone1Name=foot_l
  bone2Name=calf_l
  swing1LimitDegrees=45
  swing1Motion=Free
  swing2LimitDegrees=45
  swing2Motion=Free
  twistLimitDegrees=45
  twistMotion=Free
 }
 {
  bone1Name=foot_r
  bone2Name=calf_r
  swing1LimitDegrees=45
  swing1Motion=Free
  swing2LimitDegrees=45
  swing2Motion=Free
  twistLimitDegrees=45
  twistMotion=Free
 }
 {
  bone1Name=dyn_coat_l_02
  bone2Name=pelvis
  swing1LimitDegrees=45
  swing1Motion=Free
  swing2LimitDegrees=45
  swing2Motion=Free
  twistLimitDegrees=45
  twistMotion=Free
 }
 {
  bone1Name=dyn_coat_l_03
  bone2Name=dyn_coat_l_02
  swing1LimitDegrees=45
  swing1Motion=Free
  swing2LimitDegrees=45
  swing2Motion=Free
  twistLimitDegrees=45
  twistMotion=Free
 }
 {
  bone1Name=dyn_coat_r_02
  bone2Name=pelvis
  swing1LimitDegrees=45
  swing1Motion=Free
  swing2LimitDegrees=45
  swing2Motion=Free
  twistLimitDegrees=45
  twistMotion=Free
 }
 {
  bone1Name=dyn_coat_r_03
  bone2Name=dyn_coat_r_02
  swing1LimitDegrees=45
  swing1Motion=Free
  swing2LimitDegrees=45
  swing2Motion=Free
  twistLimitDegrees=45
  twistMotion=Free
 }
]
