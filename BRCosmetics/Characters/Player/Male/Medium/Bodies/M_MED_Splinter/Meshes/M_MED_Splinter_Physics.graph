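# Physics asset for the male medium Splinter cosmetic body, covering its dynamic fur and hood bones on the upper body.
# Capsules on the spine, clavicles and head form the base, with back and side fur jointed to the upper spine under narrow swing and twist limits.
# Two-link shoulder fur chains hang from small MSTR bodies, whose clavicle joints rotate freely.
# The hood-back joint to spine_05 has all angular motion locked.
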PhysicsAsset /BRCosmetics/Characters/Player/Male/Medium/Bodies/M_MED_Splinter/Meshes/M_MED_Splinter_Physics
bodies {
 clavicle_l [{shapeType=Capsule, length=4, radius=1}]
 clavicle_r [{shapeType=Capsule, length=4, radius=1}]
 dyn_fur_bk_01 [{shapeType=Capsule, length=3, radius=1, center=<0.1 0 0>, rotation=<rot 0 0 90>}]
 dyn_fur_bk_02 [{shapeType=Capsule, length=3, radius=1, center=<0.1 0 0>, rotation=<rot 0 0 90>}]
 dyn_fur_l_02 [{shapeType=Capsule, length=3, radius=1, center=<0.2556 0.2181 0>, rotation=<rot 0 0 90>}]
 dyn_fur_l_03 [{shapeType=Capsule, length=3, radius=1, center=<0.1 0 0>, rotation=<rot 0 0 90>}]
 dyn_fur_r_02 [{shapeType=Capsule, length=3, radius=1, center=<0.4516 0.0454 0>, rotation=<rot 0 0 90>}]
 dyn_fur_r_03 [{shapeType=Capsule, length=3, radius=1, center=<0.2445 -0.1513 0>, rotation=<rot 0 0 90>}]
 dyn_fur_shldr_l_01 [{shapeType=Capsule, length=3, radius=1, center=<0.1 0 0>, rotation=<rot 0 0 90>}]
 dyn_fur_shldr_l_01_MSTR [{shapeType=Capsule, length=0.505, radius=0.505, rotation=<rot 0 -90 0>}]
 dyn_fur_shldr_l_02 [{shapeType=Capsule, length=2.4137, radius=1.5863, center=<0.1 0 0>, rotation=<rot 2.1185 1.6003 91.0644>}]
 dyn_fur_shldr_r_01 [{shapeType=Capsule, length=3, radius=1, center=<0.1 0 0>, rotation=<rot 0 0 90>}]
 dyn_fur_shldr_r_01_MSTR [{shapeType=Capsule, length=0.505, radius=0.505, rotation=<rot 0 -90 0>}]
 dyn_fur_shldr_r_02 [{shapeType=Capsule, length=2.4137, radius=1.5863, center=<0.1 0 0>, rotation=<rot 2.1185 1.6003 91.0644>}]
 dyn_hood_bk [{shapeType=Capsule, length=3, radius=1, center=<0.1 0 0>, rotation=<rot 0 0 90>}]
 head [{shapeType=Capsule, length=4.0784, radius=9.8603, rotation=<rot -0.6877 0 0>}]
 spine_04 [{shapeType=Capsule, length=4, radius=1}]
 spine_05 [{shapeType=Capsule, length=4, radius=1}]
}
constraints [
 {bone1Name=spine_05, bone2Name=spine_04, swing1LimitDegrees=45, swing2LimitDegrees=45, twistLimitDegrees=45}
 {bone1Name=clavicle_l, bone2Name=spine_05, swing1LimitDegrees=45, swing2LimitDegrees=45, twistLimitDegrees=45}
 {bone1Name=clavicle_r, bone2Name=spine_05, swing1LimitDegrees=45, swing2LimitDegrees=45, twistLimitDegrees=45}
 {bone1Name=head, bone2Name=spine_05, swing1LimitDegrees=45, swing2LimitDegrees=45, twistLimitDegrees=45}
 {bone1Name=dyn_fur_shldr_r_01, bone2Name=dyn_fur_shldr_r_01_MSTR, swing1LimitDegrees=20, swing2LimitDegrees=20, twistLimitDegrees=5}
 {bone1Name=dyn_fur_shldr_r_02, bone2Name=dyn_fur_shldr_r_01, swing1LimitDegrees=25, swing2LimitDegrees=25, twistLimitDegrees=5}
 {
  bone1Name=dyn_hood_bk
  bone2Name=spine_05
  swing1LimitDegrees=45
  swing1Motion=Locked
  swing2LimitDegrees=45
  swing2Motion=Locked
  twistLimitDegrees=15
  twistMotion=Locked
 }
 {bone1Name=dyn_fur_bk_01, bone2Name=spine_05, swing1LimitDegrees=30, swing2LimitDegrees=30, twistLimitDegrees=5}
 {bone1Name=dyn_fur_r_02, bone2Name=spine_05, swing1LimitDegrees=30, swing2LimitDegrees=30, twistLimitDegrees=5}
 {bone1Name=dyn_fur_l_02, bone2Name=spine_05, swing1LimitDegrees=30, swing2LimitDegrees=30, twistLimitDegrees=5}
 {bone1Name=dyn_fur_r_03, bone2Name=spine_04, swing1LimitDegrees=30, swing2LimitDegrees=30, twistLimitDegrees=5}
 {bone1Name=dyn_fur_l_03, bone2Name=spine_04, swing1LimitDegrees=30, swing2LimitDegrees=30, twistLimitDegrees=5}
 {bone1Name=dyn_fur_shldr_l_01, bone2Name=dyn_fur_shldr_l_01_MSTR, swing1LimitDegrees=10, swing2LimitDegrees=10, twistLimitDegrees=5}
 {bone1Name=dyn_fur_shldr_l_02, bone2Name=dyn_fur_shldr_l_01, swing1LimitDegrees=30, swing2LimitDegrees=30, twistLimitDegrees=5}
 {bone1Name=dyn_fur_bk_02, bone2Name=spine_05, swing1LimitDegrees=35, swing2LimitDegrees=10, twistLimitDegrees=5}
 {
  bone1Name=dyn_fur_shldr_l_01_MSTR
  bone2Name=clavicle_l
  swing1LimitDegrees=45
  swing1Motion=Free
  swing2LimitDegrees=45
  swing2Motion=Free
  twistLimitDegrees=45
  twistMotion=Free
 }
 {
  bone1Name=dyn_fur_shldr_r_01_MSTR
  bone2Name=clavicle_r
  swing1LimitDegrees=45
  swing1Motion=Free
  swing2LimitDegrees=45
  swing2Motion=Free
  twistLimitDegrees=45
  twistMotion=Free
 }
]
