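# Animation Blueprint for the M_MED_Splinter character body that adds secondary motion and pose correctives to its coat, hood and shoulder fur.
# The copied parent pose passes through constraints, AnimDynamics and bone offsets into a cached pose.
# That cache feeds two chained RigidBody simulations with skydiving force, or thigh-driven trench coat rotations when RigidBody is disabled.
# The EventGraph derives corrective alphas from thigh and head ref-pose deltas, plus an idle-crouch flag.
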AnimBlueprint /BRCosmetics/Characters/Player/Male/Medium/Bodies/M_MED_Splinter/Meshes/M_MED_Splinter_AnimBP
target_skeleton=&/M_MED_Splinter_Skeleton
graphs [
 AnimGraph {nodes [
   n1 = Get thigh_out_l() -> (thigh_out_l:double)
   n2 = "Transform (Modify) Bone - Bone: dyn_trenchCoat_fl_MSTR"(
    ComponentPose=Local To Component(LocalPose=Use cached pose 'pre_dynamcis'())
    Alpha=@n1
    Translation="0,0,0"
    Rotation="20,0,-40"
    Scale="1,1,1"
   ) {
    rotationMode=BMM_Additive
    translationMode=BMM_Additive
    alphaBoolBlend {blendOption=Linear}
    boneToModify {boneName=dyn_trenchCoat_fl_MSTR}
    rotation {pitch=20, roll=-40}
   }
   n3 = "Transform (Modify) Bone - Bone: dyn_trenchCoat_l_MSTR"(ComponentPose=@n2, Alpha=@n1, Translation="0,0,0", Rotation="45,0.000,0.000", Scale="1,1,1") {
    rotationMode=BMM_Additive
    translationMode=BMM_Additive
    alphaBoolBlend {blendOption=Linear}
    boneToModify {boneName=dyn_trenchCoat_l_MSTR}
    rotation {pitch=45}
   }
   n4 = Get thigh_out_r() -> (thigh_out_r:double)
   n5 = "Transform (Modify) Bone - Bone: dyn_trenchCoat_fr_MSTR"(ComponentPose=@n3, Alpha=@n4, Translation="0,0,0", Rotation="-40,0,-30", Scale="1,1,1") {
    rotationMode=BMM_Additive
    translationMode=BMM_Additive
    alphaBoolBlend {blendOption=Linear}
    boneToModify {boneName=dyn_trenchCoat_fr_MSTR}
    rotation {pitch=-40, roll=-30}
   }
   n6 = "Transform (Modify) Bone - Bone: dyn_trenchCoat_r_MSTR"(ComponentPose=@n5, Alpha=@n4, Translation="0,0,0", Rotation="-45,0.000,0.000", Scale="1,1,1") {
    rotationMode=BMM_Additive
    translationMode=BMM_Additive
    alphaBoolBlend {blendOption=Linear}
    boneToModify {boneName=dyn_trenchCoat_r_MSTR}
    rotation {pitch=-45}
   }
   n7 = Break Fort Anim Input Skydiving External Force(FortAnimInput_SkydivingExternalForce=Get SkydivingExternalForceInput()) -> (
    bUseSkydivingVectorForce=~
    SkydivingVectorForce=~
    HeadToPelvisDirection=~
    FloatingMultiplier=~
    FloatingAdditive=~
    DivingMultiplier=~
    DivingAdditive=~
    ParachutingMultiplier=~
    ParachutingAdditive=~
   )
   n8 = RigidBody(ComponentPose=Local To Component(LocalPose=Use cached pose 'pre_dynamcis'()), ExternalForce=@n7.SkydivingVectorForce) {
    bClampLinearTranslationLimitToRefPose=true
    bOverrideWorldGravity=true
    overridePhysicsAsset=&/M_MED_Splinter_Physics
    simulationSpace=BaseBoneSpace
    alphaBoolBlend {blendOption=Linear}
    baseBoneRef {boneName=spine_03}
    componentLinearAccScale=<0.5 0.5 0.3>
    componentLinearVelScale=<0.5 0.5 0.3>
    overrideWorldGravity {z=-980}
   }
   n9 = RigidBody(ComponentPose=@n8, ExternalForce=@n7.SkydivingVectorForce) {
    bClampLinearTranslationLimitToRefPose=true
    bOverrideWorldGravity=true
    overridePhysicsAsset=&/M_MED_Splinter_Jacket_Physics
    simulationSpace=BaseBoneSpace
    alphaBoolBlend {blendOption=Linear}
    baseBoneRef {boneName=pelvis}
    componentLinearAccScale=<0.4 0.2 0.2>
    componentLinearVelScale=<0.4 0.2 0.2>
    overrideWorldGravity {z=-980}
   }
   Output Pose(Result=Blend Poses by bool(BlendPose_0=Component To Local(ComponentPose=@n9), BlendPose_1=Component To Local(ComponentPose=@n6), bActiveValue=Get bRigidBodyEnabled()) {blendType=Linear}) {layerGroup=None}
   n10 = Get thigh_up_l() -> (thigh_up_l:double)
   n11 = "Constraint - dyn_coat_lb_01 (R)"(ComponentPose=Local To Component(LocalPose=Copy Pose From Mesh() {bUseAttachedParent=true}), Alpha=0, ConstraintWeights_0=1) {
    alphaBoolBlend {blendOption=Linear}
    boneToModify {boneName=dyn_coat_lb_01}
    constraintSetup [{transformType=Rotation, targetBone {boneName=thigh_l}}]
    constraintWeights [1]
   }
   n12 = "Constraint - dyn_coat_rb_01 (R)"(ComponentPose=@n11, Alpha=0, ConstraintWeights_0=1) {
    alphaBoolBlend {blendOption=Linear}
    boneToModify {boneName=dyn_coat_rb_01}
    constraintSetup [{transformType=Rotation, targetBone {boneName=thigh_r}}]
    constraintWeights [1]
   }
   n13 = Get bIsPlayingEmote()
   n14 = "Constraint - dyn_coat_l_MSTR (PPPPPPP)"!(
    ComponentPose=@n12
    bAlphaBoolEnabled:bool=@n13
    ConstraintWeights_0=0.0806
    ConstraintWeights_1:float=0.0069
    ConstraintWeights_2:float=0.0011
    ConstraintWeights_3:float=0.3605
    ConstraintWeights_4:float=0.4831
    ConstraintWeights_5:float=0.0663
    ConstraintWeights_6:float=0.0014
   ) -> (Pose) {
    alphaInputType=Bool
    alphaBoolBlend {blendInTime=0.1, blendOption=Linear, blendOutTime=0.1}
    boneToModify {boneName=dyn_coat_l_MSTR}
    constraintSetup [
     {targetBone {boneName=spine_01}}
     {targetBone {boneName=spine_02}}
     {targetBone {boneName=spine_03}}
     {targetBone {boneName=thigh_l}}
     {targetBone {boneName=pelvis}}
     {targetBone {boneName=thigh_twist_01_l}}
     {targetBone {boneName=thigh_twist_01_r}}
    ]
    constraintWeights [1,1,1,1,1,1,1]
   }
   n15 = "Transform (Modify) Bone - Bone: dyn_coat_l_MSTR"!(ComponentPose=@n14, bAlphaBoolEnabled:bool=@n13, Translation="5,-1,2", Rotation="0,0,0", Scale="1,1,1") -> (Pose) {
    alphaInputType=Bool
    translationMode=BMM_Additive
    translationSpace=BCS_BoneSpace
    alphaBoolBlend {blendInTime=0.1, blendOption=Linear, blendOutTime=0.1}
    boneToModify {boneName=dyn_coat_l_MSTR}
    translation=<5 -1 2>
   }
   n16 = "Constraint - dyn_coat_r_MSTR (PPPPPPP)"!(
    ComponentPose=@n15
    bAlphaBoolEnabled:bool=@n13
    ConstraintWeights_0=0.076
    ConstraintWeights_1:float=0.0523
    ConstraintWeights_2:float=0.0238
    ConstraintWeights_3:float=0.3211
    ConstraintWeights_4:float=0.4506
    ConstraintWeights_5:float=0.003
    ConstraintWeights_6:float=0.0733
   ) -> (Pose) {
    alphaInputType=Bool
    alphaBoolBlend {blendInTime=0.1, blendOption=Linear, blendOutTime=0.1}
    boneToModify {boneName=dyn_coat_r_MSTR}
    constraintSetup [
     {targetBone {boneName=spine_01}}
     {targetBone {boneName=spine_02}}
     {targetBone {boneName=spine_03}}
     {targetBone {boneName=thigh_r}}
     {targetBone {boneName=pelvis}}
     {targetBone {boneName=thigh_twist_01_l}}
     {targetBone {boneName=thigh_twist_01_r}}
    ]
    constraintWeights [1,1,1,1,1,1,1]
   }
   n17 = "Transform (Modify) Bone - Bone: dyn_coat_r_MSTR"!(ComponentPose=@n16, bAlphaBoolEnabled:bool=@n13, Translation="5,1,2", Rotation="0,0,0", Scale="1,1,1") -> (Pose) {
    alphaInputType=Bool
    translationMode=BMM_Additive
    translationSpace=BCS_BoneSpace
    alphaBoolBlend {blendInTime=0.1, blendOption=Linear, blendOutTime=0.1}
    boneToModify {boneName=dyn_coat_r_MSTR}
    translation=<5 1 2>
   }
   n18 = "Constraint - dyn_coat_l_MSTR (R)"(ComponentPose=@n17, Alpha=1, ConstraintWeights_0=1) {
    alphaBoolBlend {blendOption=Linear}
    boneToModify {boneName=dyn_coat_l_MSTR}
    constraintSetup [{transformType=Rotation, targetBone {boneName=thigh_l}}]
    constraintWeights [1]
   }
   n19 = "Constraint - dyn_coat_r_MSTR (R)"(ComponentPose=@n18, Alpha=1, ConstraintWeights_0=1) {
    alphaBoolBlend {blendOption=Linear}
    boneToModify {boneName=dyn_coat_r_MSTR}
    constraintSetup [{transformType=Rotation, targetBone {boneName=thigh_r}}]
    constraintWeights [1]
   }
   n20 = Anim Dynamics - Bone: dyn_coat_l_02(ComponentPose=@n19, Alpha=1, ExternalForce="0,0,0") {
    angularDampingOverride=0.9
    bGravityOverrideInSimSpace=true
    bOverrideAngularDamping=true
    bUseGravityOverride=true
    simulationSpace=RootRelative
    alphaBoolBlend {blendOption=Linear}
    boundBone {boneName=dyn_coat_l_02}
    componentAppliedLinearAccClamp=<10000 10000 10000>
    componentLinearAccScale=<0.1 0.1 0.1>
    componentLinearVelScale=<0.1 0.1 0.1>
    gravityOverride {z=-980}
    physicsBodyDefinitions [{
      sphereCollisionRadius=0
      boundBone {boneName=dyn_coat_l_02}
      boxExtents=<5 5 5>
      constraintSetup {angularLimitsMax=<0 10 5>, angularLimitsMin=<-20 -2 -5>, angularTarget=<0 0 0>}
      localJointOffset=<1 0 0>
     }]
    relativeSpaceBone {boneName=root}
   }
   n21 = Anim Dynamics - Bone: dyn_coat_l_03(ComponentPose=@n20, Alpha=1, ExternalForce="0,0,0") {
    angularDampingOverride=0.99
    bGravityOverrideInSimSpace=true
    bOverrideAngularDamping=true
    bUseGravityOverride=true
    simulationSpace=RootRelative
    alphaBoolBlend {blendOption=Linear}
    boundBone {boneName=dyn_coat_l_03}
    componentAppliedLinearAccClamp=<10000 10000 10000>
    componentLinearAccScale=<0.1 0.1 0.1>
    componentLinearVelScale=<0.1 0.1 0.1>
    gravityOverride {z=-980}
    physicsBodyDefinitions [{
      sphereCollisionRadius=0
      boundBone {boneName=dyn_coat_l_03}
      boxExtents=<5 5 5>
      constraintSetup {angularLimitsMax=<0 5 5>, angularLimitsMin=<-20 -5 -5>, angularTarget=<0 0 0>}
      localJointOffset=<1 0 0>
     }]
    relativeSpaceBone {boneName=root}
   }
   n22 = Anim Dynamics - Bone: dyn_coat_r_02(ComponentPose=@n21, Alpha=1, ExternalForce="0,0,0") {
    angularDampingOverride=0.99
    bGravityOverrideInSimSpace=true
    bOverrideAngularDamping=true
    bUseGravityOverride=true
    simulationSpace=RootRelative
    alphaBoolBlend {blendOption=Linear}
    boundBone {boneName=dyn_coat_r_02}
    componentAppliedLinearAccClamp=<10000 10000 10000>
    componentLinearAccScale=<0.1 0.1 0.1>
    componentLinearVelScale=<0.1 0.1 0.1>
    gravityOverride {z=-980}
    physicsBodyDefinitions [{
      sphereCollisionRadius=0
      boundBone {boneName=dyn_coat_r_02}
      boxExtents=<5 5 5>
      constraintSetup {angularLimitsMax=<0 5 5>, angularLimitsMin=<-20 -5 -5>, angularTarget=<0 0 0>}
      localJointOffset=<1 0 0>
     }]
   }
   n23 = Anim Dynamics - Bone: dyn_coat_r_03(ComponentPose=@n22, Alpha=1, ExternalForce="0,0,0") {
    angularDampingOverride=0.999
    bGravityOverrideInSimSpace=true
    bOverrideAngularDamping=true
    bUseGravityOverride=true
    simulationSpace=RootRelative
    alphaBoolBlend {blendOption=Linear}
    boundBone {boneName=dyn_coat_r_03}
    componentAppliedLinearAccClamp=<10000 10000 10000>
    componentLinearAccScale=<0.1 0.1 0.1>
    componentLinearVelScale=<0.1 0.1 0.1>
    gravityOverride {z=-980}
    physicsBodyDefinitions [{
      sphereCollisionRadius=0
      boundBone {boneName=dyn_coat_r_03}
      boxExtents=<5 5 5>
      constraintSetup {angularLimitsMax=<0 5 5>, angularLimitsMin=<-20 -5 -5>, angularTarget=<0 0 0>}
      localJointOffset=<1 0 0>
     }]
   }
   n24 = "Constraint - dyn_coat_lb_02 (R)"(ComponentPose=@n23, Alpha=1, ConstraintWeights_0=1) {
    alphaBoolBlend {blendOption=Linear}
    boneToModify {boneName=dyn_coat_lb_02}
    constraintSetup [{transformType=Rotation, targetBone {boneName=dyn_trenchCoat_bl_02}}]
    constraintWeights [1]
   }
   n25 = "Constraint - dyn_coat_rb_02 (R)"(ComponentPose=@n24, Alpha=1, ConstraintWeights_0=1) {
    alphaBoolBlend {blendOption=Linear}
    boneToModify {boneName=dyn_coat_rb_02}
    constraintSetup [{transformType=Rotation, targetBone {boneName=dyn_trenchCoat_br_02}}]
    constraintWeights [1]
   }
   n26 = Anim Dynamics - Bone: dyn_coat_lb_02(ComponentPose=@n25, Alpha=1, ExternalForce="0,0,0") {
    angularDampingOverride=0.9
    bGravityOverrideInSimSpace=true
    bOverrideAngularDamping=true
    bUseGravityOverride=true
    simulationSpace=RootRelative
    alphaBoolBlend {blendOption=Linear}
    boundBone {boneName=dyn_coat_lb_02}
    gravityOverride {z=-980}
    physicsBodyDefinitions [{
      sphereCollisionRadius=0
      boundBone {boneName=dyn_coat_lb_02}
      boxExtents=<5 5 5>
      constraintSetup {angularLimitsMax=<0 10 5>, angularLimitsMin=<-20 0 -5>, angularTarget=<0 0 0>}
      localJointOffset=<1 0 0>
     }]
   }
   n27 = Anim Dynamics - Bone: dyn_coat_rb_02(ComponentPose=@n26, Alpha=1, ExternalForce="0,0,0") {
    angularDampingOverride=0.99
    bGravityOverrideInSimSpace=true
    bOverrideAngularDamping=true
    bUseGravityOverride=true
    simulationSpace=RootRelative
    alphaBoolBlend {blendOption=Linear}
    boundBone {boneName=dyn_coat_rb_02}
    gravityOverride {z=-980}
    physicsBodyDefinitions [{
      sphereCollisionRadius=0
      boundBone {boneName=dyn_coat_rb_02}
      boxExtents=<5 5 5>
      constraintSetup {angularLimitsMax=<0 10 5>, angularLimitsMin=<-20 0 -5>, angularTarget=<0 0 0>}
      localJointOffset=<1 0 0>
     }]
   }
   n28 = "Constraint - dyn_coat_lb_03 (R)"(ComponentPose=@n27, Alpha=1, ConstraintWeights_0=1) {
    alphaBoolBlend {blendOption=Linear}
    boneToModify {boneName=dyn_coat_lb_03}
    constraintSetup [{transformType=Rotation, targetBone {boneName=dyn_trenchCoat_bl_03}}]
    constraintWeights [1]
   }
   n29 = "Constraint - dyn_coat_rb_03 (R)"(ComponentPose=@n28, Alpha=1, ConstraintWeights_0=1) {
    alphaBoolBlend {blendOption=Linear}
    boneToModify {boneName=dyn_coat_rb_03}
    constraintSetup [{transformType=Rotation, targetBone {boneName=dyn_trenchCoat_br_03}}]
    constraintWeights [1]
   }
   n30 = Anim Dynamics - Bone: dyn_coat_lb_03(ComponentPose=@n29, Alpha=1, ExternalForce="0,0,0") {
    angularDampingOverride=0.99
    bGravityOverrideInSimSpace=true
    bOverrideAngularDamping=true
    bUseGravityOverride=true
    simulationSpace=RootRelative
    alphaBoolBlend {blendOption=Linear}
    boundBone {boneName=dyn_coat_lb_03}
    gravityOverride {z=-980}
    physicsBodyDefinitions [{
      sphereCollisionRadius=0
      boundBone {boneName=dyn_coat_lb_03}
      boxExtents=<5 5 5>
      constraintSetup {angularLimitsMax=<0 15 5>, angularLimitsMin=<-20 0 -5>, angularTarget=<0 0 0>}
      localJointOffset=<1 0 0>
     }]
   }
   n31 = Anim Dynamics - Bone: dyn_coat_rb_03(ComponentPose=@n30, Alpha=1, ExternalForce="0,0,0") {
    angularDampingOverride=0.999
    bGravityOverrideInSimSpace=true
    bOverrideAngularDamping=true
    bUseGravityOverride=true
    simulationSpace=RootRelative
    alphaBoolBlend {blendOption=Linear}
    boundBone {boneName=dyn_coat_rb_03}
    gravityOverride {z=-980}
    physicsBodyDefinitions [{
      sphereCollisionRadius=0
      boundBone {boneName=dyn_coat_rb_03}
      boxExtents=<5 5 5>
      constraintSetup {angularLimitsMax=<0 15 5>, angularLimitsMin=<-20 0 -5>, angularTarget=<0 0 0>}
      localJointOffset=<1 0 0>
     }]
   }
   n32 = "Transform (Modify) Bone - Bone: dyn_coat_lb_03"(ComponentPose=@n31, Alpha=1, Translation="0,0,0", Rotation="0.000,-10,-10.000", Scale="1,1,1") {rotationMode=BMM_Additive, alphaBoolBlend {blendOption=Linear}, boneToModify {boneName=dyn_coat_lb_03}, rotation {roll=-10, yaw=-10}}
   n33 = "Transform (Modify) Bone - Bone: dyn_coat_rb_03"(ComponentPose=@n32, Alpha=1, Translation="0,0,0", Rotation="0.000,10,-10", Scale="1,1,1") {rotationMode=BMM_Additive, alphaBoolBlend {blendOption=Linear}, boneToModify {boneName=dyn_coat_rb_03}, rotation {roll=-10, yaw=10}}
   n34 = "Transform (Modify) Bone - Bone: dyn_trenchCoat_fl_MSTR"(ComponentPose=@n33, Alpha=@n10, Translation="10,0,0", Rotation="0.000,-55,0.000", Scale="1,1,1") {
    rotationMode=BMM_Additive
    translationMode=BMM_Additive
    alphaBoolBlend {blendOption=Linear}
    boneToModify {boneName=dyn_trenchCoat_fl_MSTR}
    rotation {yaw=-55}
    translation {x=10}
   }
   n35 = "Transform (Modify) Bone - Bone: dyn_coat_l_MSTR"!(ComponentPose=@n34, Alpha:float=@n10, Translation="0,-2,2", Rotation="0.000,-10,0.000", Scale="1,1,1") -> (Pose) {
    rotationMode=BMM_Additive
    translationMode=BMM_Additive
    alphaBoolBlend {blendOption=Linear}
    boneToModify {boneName=dyn_coat_l_MSTR}
    rotation {yaw=-10}
    translation {y=-2, z=2}
   }
   n36 = "Transform (Modify) Bone - Bone: dyn_trenchCoat_br_MSTR"!(ComponentPose=@n35, bAlphaBoolEnabled:bool=Get bIsMoving2D(), Translation="0,0,0", Rotation="-0.000,0.000,20", Scale="1,1,1") -> (Pose) {
    alphaInputType=Bool
    rotationMode=BMM_Additive
    translationMode=BMM_Additive
    alphaBoolBlend {blendInTime=0.2, blendOption=Linear, blendOutTime=0.2}
    boneToModify {boneName=dyn_trenchCoat_br_MSTR}
    rotation {roll=20}
   }
   n37 = Get idleCrouch() -> (idleCrouch:bool)
   n38 = "Transform (Modify) Bone - Bone: dyn_trenchCoat_r_MSTR"!(ComponentPose=@n36, bAlphaBoolEnabled:bool=@n37, Translation="0,0,0", Rotation="5.946,-71.840,28.340", Scale="1,1,1") -> (Pose) {
    alphaInputType=Bool
    rotationMode=BMM_Additive
    translationMode=BMM_Additive
    alphaBoolBlend {blendInTime=0.2, blendOption=Linear, blendOutTime=0.2}
    boneToModify {boneName=dyn_trenchCoat_r_MSTR}
    rotation=<rot 5.946 -71.84 28.34>
   }
   n39 = "Transform (Modify) Bone - Bone: dyn_trenchCoat_fr_MSTR"!(ComponentPose=@n38, bAlphaBoolEnabled:bool=@n37, Translation="0,0,0", Rotation="-30,0.000,-0.000", Scale="1,1,1") -> (Pose) {
    alphaInputType=Bool
    rotationMode=BMM_Additive
    translationMode=BMM_Additive
    alphaBoolBlend {blendInTime=0.2, blendOption=Linear, blendOutTime=0.2}
    boneToModify {boneName=dyn_trenchCoat_fr_MSTR}
    rotation {pitch=-30}
   }
   n40 = "Transform (Modify) Bone - Bone: dyn_trenchCoat_fr_MSTR"!(ComponentPose=@n39, Alpha:float=Get thigh_up_r() -> (thigh_up_r:double), Translation="-5,0,0", Rotation="0.000,50,0.000", Scale="1,1,1") -> (Pose) {
    rotationMode=BMM_Additive
    translationMode=BMM_Additive
    alphaBoolBlend {blendOption=Linear}
    boneToModify {boneName=dyn_trenchCoat_fr_MSTR}
    rotation {yaw=50}
    translation {x=-5}
   }
   n41 = "Transform (Modify) Bone - Bone: dyn_fur_shldr_r_01_MSTR"!(ComponentPose=@n40, bAlphaBoolEnabled:bool=Get bIsTargeting(), Translation="-3,0,0", Rotation="0,0,0", Scale="1,1,1") -> (Pose) {
    alphaInputType=Bool
    rotationMode=BMM_Additive
    translationMode=BMM_Additive
    alphaBoolBlend {blendOption=Linear}
    boneToModify {boneName=dyn_fur_shldr_r_01_MSTR}
    translation {x=-3}
   }
   n42 = "Transform (Modify) Bone - Bone: hood_side_l"(ComponentPose=@n41, Alpha=Get head_rot_l() -> (head_rot_l:double), Translation="0,0,0", Rotation="5,-45,-5", Scale="1,1,1") {rotationMode=BMM_Additive, alphaBoolBlend {blendOption=Linear}, boneToModify {boneName=hood_side_l}, rotation=<rot 5 -45 -5>}
   n43 = "Transform (Modify) Bone - Bone: hood_chin_l"(ComponentPose=@n42, Alpha=Get look_l_chin() -> (look_l_chin:double), Translation="3,5,0.000", Rotation="0,0,0", Scale="1,1,1") {translationMode=BMM_Additive, alphaBoolBlend {blendOption=Linear}, boneToModify {boneName=hood_chin_l}, translation=<3 5>}
   n44 = "Transform (Modify) Bone - Bone: hood_side_r"(ComponentPose=@n43, Alpha=Get head_rot_r() -> (head_rot_r:double), Translation="0,0,0", Rotation="0.000,45,0.000", Scale="1,1,1") {rotationMode=BMM_Additive, alphaBoolBlend {blendOption=Linear}, boneToModify {boneName=hood_side_r}, rotation {yaw=45}}
   Save cached pose 'pre_dynamcis'(Pose=Component To Local(ComponentPose=@n44)) {cacheName=pre_dynamcis}
  ]}
 EventGraph {nodes [
   Event BlueprintUpdateAnimation()
   n1 = BreakTransform(InTransform=GetDeltaTransformFromRefPose(self=GetOwnerSkeletalMeshComponent(), BoneName=thigh_l, BaseName=pelvis))
   > Set thigh_up_l(thigh_up_l:double=MapRangeClamped(Value=BreakRotator(InRot=@n1.Rotation).Yaw, InRangeA=-20, InRangeB=-100)) -> (Output_Get:double)
   > Set thigh_out_l(thigh_out_l:double=MapRangeClamped(Value=BreakRotator(InRot=@n1.Rotation).Pitch, InRangeA=0, InRangeB=40)) -> (Output_Get:double)
   n2 = BreakTransform(InTransform=GetDeltaTransformFromRefPose(self=GetOwnerSkeletalMeshComponent(), BoneName=thigh_r, BaseName=pelvis))
   > Set thigh_up_r(thigh_up_r:double=MapRangeClamped(Value=BreakRotator(InRot=@n2.Rotation).Yaw, InRangeA=-20, InRangeB=-100)) -> (Output_Get:double)
   > Set thigh_out_r(thigh_out_r:double=MapRangeClamped(Value=BreakRotator(InRot=@n2.Rotation).Pitch, InRangeA=0, InRangeB=35)) -> (Output_Get:double)
   n3 = GetOwnerSkeletalMeshComponent()
   n4 = BreakRotator(InRot=BreakTransform(InTransform=GetDeltaTransformFromRefPose(self=@n3, BoneName=head, BaseName=neck_02)).Rotation)
   > Set head_rot_r(head_rot_r:double=MapRangeClamped(Value=@n4, InRangeA=0, InRangeB=-20)) -> (Output_Get:double)
   > Set head_rot_l(head_rot_l:double=MapRangeClamped(Value=@n4, InRangeA=0, InRangeB=20)) -> (Output_Get:double)
   > Set look_l_chin(look_l_chin:double=MapRangeClamped(Value=@n4, InRangeA=0, InRangeB=40)) -> (Output_Get:double)
   n5 = Get bIsMoving2D()
   n6 = "Absolute (Float)"(A=BreakRotator(InRot=Get PawnDeltaRotation() -> (PawnDeltaRotation:Rotator)).Yaw)
   > Set static_crouching_rotation(static_crouching_rotation:bool=AND Boolean!(A="float > float"(A=@n6, B=2), B=Get bIsCrouching(), C:bool=NOT Boolean(A=@n5)) -> (ReturnValue)) -> (Output_Get:bool)
   n7 = Get static_crouching_rotation() -> (static_crouching_rotation:bool)
   n8 = Get bIsMoving2D()
   > Set idleCrouch(idleCrouch:bool=AND Boolean(A=NOT Boolean(A=@n8), B=Get bIsCrouching(), C=NOT Boolean(A=@n7))) -> (Output_Get:bool)
  ]}
]
variables [
 head_rot_l {type=number}
 head_rot_r {type=number}
 idleCrouch {type=boolean}
 look_l_chin {type=number}
 static_crouching_rotation {type=boolean}
 thigh_out_l {type=number}
 thigh_out_r {type=number}
 thigh_up_l {type=number}
 thigh_up_r {type=number}
]
